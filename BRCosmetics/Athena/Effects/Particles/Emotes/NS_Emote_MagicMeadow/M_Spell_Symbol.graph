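# Additive unlit, two-sided Niagara sprite material for a symbol effect in the Magic Meadow emote.
# A grayscale main texture is tinted by particle color, with UVs optionally warped by two panning noise textures.
# Opacity combines particle alpha or noise-based erosion driven by it, a dynamic sprite-opacity parameter and an optional double-sided Fresnel fade.
# A switchable world position offset adds sine and cosine wobble on three axes.
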Material /BRCosmetics/Athena/Effects/Particles/Emotes/NS_Emote_MagicMeadow/M_Spell_Symbol
graph {
 properties {
  bUsedWithNiagaraSprites=true
  blendMode=BLEND_Additive
  pixelDepthOffsetMode=PDOM_Legacy
  shadingModel=MSM_Unlit
  twoSided=true
  usedShadingModels=Unlit
 }
 nodes [
  n1 = ScalarParameter() {parameterName=TwistAmount_Thin}
  n2 = ParticleRelativeTime()
  n3 = ScalarParameter() {parameterName=U_Speed}
  n4 = ScalarParameter() {defaultValue=-0.5, parameterName=V_speed}
  Add(
   A=Multiply(
    A=MaterialFunctionCall!(UV, Twist Center=@n1, Twist Edges) {materialFunction=/Game/Accessories/FORT_Backpacks/Backpack_F_MED_Celestial_Glow/FX/Materials/MF_PolarUV}.Polar UVs
    B=ScalarParameter() {defaultValue=2, parameterName=Param_5}
   )
   B=AppendVector(A=Multiply(A=@n2, B=@n3), B=Multiply(A=@n2, B=@n4))
  ) {constB=0}
  DynamicParameter() {defaultValue=<rgba 1 0 0 0>, paramNames [Sprite_Opacity, Wiggle_Frequency, Wiggle_Amplitude_1, Wiggle_Amplitude_2]}
  n5 = StaticSwitchParameter(
   True=MaterialFunctionCall() {materialFunction=&MF_PolarUV}.Polar UVs
   False=TextureCoordinate()
  ) {parameterName=Use_Distortion_Polar_Uvs}
  n6 = ScalarParameter() {defaultValue=0.7, group=Distortion Tex 1, parameterName=Distortion_1_Scale}
  n7 = ScalarParameter() {defaultValue=0.3, group=Distortion Tex 1, parameterName=Distortion_1_PanSpeed_X}
  n8 = ScalarParameter() {defaultValue=0.6, group=Distortion Tex 1, parameterName=Distortion_1_PanSpeed_Y}
  n9 = TextureSampleParameter2D(UVs=Panner(Coordinate=Multiply(A=@n5, B=@n6), Time=Time(), Speed=AppendVector(A=@n7, B=@n8)) {speedX=0.4, speedY=0.5}) {
   group=Distortion Tex 1
   parameterName=Tex_Distortion
   samplerType=SAMPLERTYPE_LinearGrayscale
   texture=&T_SimplePerlin_Greyscale
  }
  n10 = ScalarParameter() {defaultValue=0.6, group=Distortion Tex 2, parameterName=Distortion_2_Scale}
  n11 = ScalarParameter() {defaultValue=-0.2, group=Distortion Tex 2, parameterName=Distortion_2_PanSpeed_X}
  n12 = ScalarParameter() {defaultValue=-0.4, group=Distortion Tex 2, parameterName=Distortion_2_PanSpeed_Y}
  n13 = TextureSampleParameter2D(UVs=Panner(Coordinate=Multiply(A=@n5, B=@n10) {constB=0.6}, Time=Time(), Speed=AppendVector(A=@n11, B=@n12)) {speedX=-0.2, speedY=-0.4}) {
   group=Distortion Tex 2
   parameterName=Tex_Distortion2
   samplerType=SAMPLERTYPE_Grayscale
   texture=&LowResBlurredNoise
  }
  n14 = Multiply(A=LinearInterpolate(A=@n9.R, B=@n13.R), B=ScalarParameter() {defaultValue=0.08, group=Distortion, parameterName=Distort_Power}) {constB=0.08}
  n15 = TextureCoordinate()
  n16 = TextureCoordinate()
  n17 = TextureSampleParameter2D(UVs=StaticSwitchParameter(True=Add(A=@n14, B=@n15), False=@n16) {defaultValue=true, parameterName=Use_Distort}) {
   group=Main
   parameterName=Tex1
   samplerType=SAMPLERTYPE_LinearGrayscale
   texture=&/T_Igni_Main
  }
  n18 = ParticleColor()
  MaterialFunctionCall(Emissive=Multiply(A=@n17.RGB, B=@n18.RGB)) {materialFunction=&MF_FX_EmissiveAdjustments}
  n19 = Time()
  MaterialFunctionCall(X=@n19, Y=@n19) {materialFunction=&MakeFloat2}
  n20 = Constant() {r=1}
  n21 = DynamicParameter() {defaultValue=<rgba 1 0 0 0>, paramNames [Sprite_Opacity, Wiggle_Frequency, Wiggle_Amplitude_1, Wiggle_Amplitude_2]}
  n22 = ScalarParameter() {defaultValue=1, parameterName=Use_Sprite_Opacity_Parameter}
  n23 = TextureSampleParameter2D(UVs=Multiply(A=@n16, B=ScalarParameter() {defaultValue=2, group=Erosion, parameterName=Erosion_Scale}) {constB=2}) {group=Erosion, parameterName=Tex_Erosion, texture=/Game/Items/Traps/Textures/T_Barrier_Clouds_03_M}
  n24 = Multiply(
   A=@n17.RGB
   B=MaterialFunctionCall!(Gradient=Saturate(None=@n23.R), Erosion Amount=@n18.A, Softness) {materialFunction=/BRCosmetics/Athena/Effects/Particles/Skydiving/NS_Contrail_DarkAndLight/Functions/MF_Erosion}.Increased Range
  )
  n25 = Multiply(A=@n17.RGB, B=@n18.A)
  n26 = ScalarParameter() {defaultValue=1, group=Alpha, parameterName="CameraNormal_Fresnel_Brightness (S)"}
  n27 = ScalarParameter() {defaultValue=1, group=Alpha, parameterName="CameraNormal_Fresnel_Exponent (S)"}
  n28 = VertexNormalWS()
  Multiply(
   A=Multiply(
    A=LinearInterpolate(A=@n20, B=@n21.Sprite_Opacity, Alpha=@n22)
    B=StaticSwitchParameter(True=@n24, False=@n25) {defaultValue=true, parameterName=Use_Dissolve}
   )
   B=StaticSwitchParameter(
    True=MaterialFunctionCall(Brightness=@n26, Exponent=@n27, Vertex / Pixel Normal=@n28) {materialFunction=&MF_Fresnel_FX}.Fresnel Double Sided
    False=Constant() {r=1}
   ) {defaultValue=true, parameterName=Side_Fade}
  )
  n29 = Add(A=Time()) {constB=0}
  n30 = Multiply(A=@n29) {constB=0.2}
  n31 = Multiply(A=Constant3Vector() {constant=<rgba 1 0 0 0>}, B=Multiply(A=Sine(None=@n30)) {constB=3}) {constB=0}
  n32 = Multiply(A=@n29) {constB=0.3}
  n33 = Multiply(A=Multiply(A=Cosine(None=@n32) {period=1.2}) {constB=1.3}, B=Constant3Vector() {constant=<rgba 0 1 0 0>})
  n34 = Multiply(A=Sine(None=Multiply(A=@n29) {constB=0.55}) {period=-0.8}) {constB=3.5}
  n35 = Constant3Vector() {constant=<rgba 0 0 1 0>}
  StaticSwitchParameter(True=Add(A=Add(A=@n31, B=@n33), B=Multiply(A=@n34, B=@n35)), False=Constant()) {defaultValue=true, parameterName=Use_WPO}
  TextureCoordinate()
  TextureCoordinate()
  Time()
 ]
}
