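# Additive, unlit, two-sided Niagara mesh-particle material for the sides of the spell symbol in the Magic Meadow emote.
# Two borrowed pickaxe trail textures scroll oppositely over particle lifetime, warped by panning Perlin noise, then blend and take particle color through emissive adjustments.
# An opacity chain multiplies the blend by particle alpha, U-edge fades and a switchable double-sided Fresnel.
# The graph also builds a normal-offset and sine/cosine wobble chain, without showing its destination.
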Material /BRCosmetics/Athena/Effects/Particles/Emotes/NS_Emote_MagicMeadow/M_Spell_Symbol_Sides
graph {
 properties {
  bUsedWithNiagaraMeshParticles=true
  blendMode=BLEND_Additive
  pixelDepthOffsetMode=PDOM_Legacy
  shadingModel=MSM_Unlit
  twoSided=true
  usedShadingModels=Unlit
 }
 nodes [
  n1 = TextureCoordinate()
  n2 = ScalarParameter() {defaultValue=0.7, group=Distortion Tex 1, parameterName=Distortion_1_Scale}
  n3 = ScalarParameter() {defaultValue=0.5, group=Distortion Tex 1, parameterName=Distortion_1_PanSpeed_X}
  n4 = ScalarParameter() {defaultValue=0.3, group=Distortion Tex 1, parameterName=Distortion_1_PanSpeed_Y}
  n5 = TextureSampleParameter2D(UVs=Panner(Coordinate=Multiply(A=@n1, B=@n2), Time=Time(), Speed=AppendVector(A=@n3, B=@n4)) {speedX=0.4, speedY=0.5}) {
   group=Distortion Tex 1
   parameterName=Tex_Distortion
   samplerType=SAMPLERTYPE_LinearGrayscale
   texture=&T_SimplePerlin_Greyscale
  }
  n6 = Multiply(A=@n5.R, B=ScalarParameter() {defaultValue=0.2, group=Distortion, parameterName=Distort_Power}) {constB=0.08}
  n7 = TextureCoordinate()
  n8 = AppendVector(A=Multiply(A=ComponentMask(None=@n7) {r=true}) {constB=6}, B=Multiply(A=ComponentMask(None=@n7) {g=true}))
  n9 = ParticleRelativeTime()
  n10 = Constant2Vector() {r=7}
  n11 = TextureSampleParameter2D(UVs=Add(A=@n6, B=Panner(Coordinate=@n8, Time=@n9, Speed=@n10))) {
   parameterName=Tex1
   texture=/BRCosmetics/Weapons/FORT_Melee/Pickaxe_Future_Samurai_Summer_Male/FX/T_Pickaxe_Future_Samurai_Summer_Male_Trail_01
  }
  n12 = TextureCoordinate()
  n13 = AppendVector(A=Multiply(A=ComponentMask(None=@n12) {r=true}) {constB=6}, B=Multiply(A=ComponentMask(None=@n12) {g=true}))
  n14 = ParticleRelativeTime()
  n15 = Constant2Vector() {r=-5}
  n16 = TextureSampleParameter2D(UVs=Add(A=@n6, B=Panner(Coordinate=@n13, Time=@n14, Speed=@n15))) {
   parameterName=Tex2
   texture=/BRCosmetics/Weapons/FORT_Melee/Pickaxe_Future_Samurai_Summer_Male/FX/T_Pickaxe_Future_Samurai_Summer_Male_Trail_02
  }
  n17 = LinearInterpolate(A=@n11.R, B=@n16.R) {constAlpha=0.6}
  n18 = Saturate(None=@n17)
  n19 = Add(A=Time()) {constB=0}
  n20 = Multiply(A=@n19) {constB=0.2}
  n21 = Multiply(A=Constant3Vector() {constant=<rgba 1 0 0 0>}, B=Multiply(A=Sine(None=@n20)) {constB=3}) {constB=0}
  n22 = Multiply(A=@n19) {constB=0.3}
  n23 = Multiply(A=Multiply(A=Cosine(None=@n22) {period=1.2}) {constB=1.3}, B=Constant3Vector() {constant=<rgba 0 1 0 0>})
  n24 = Multiply(A=Sine(None=Multiply(A=@n19) {constB=0.55}) {period=-0.8}) {constB=3.5}
  n25 = Constant3Vector() {constant=<rgba 0 0 1 0>}
  Add(A=Multiply(A=VertexNormalWS(), B=Multiply(A=@n18) {constB=0.5}), B=Add(A=Add(A=@n21, B=@n23), B=Multiply(A=@n24, B=@n25)))
  DynamicParameter() {defaultValue=<rgba 0 0 0 0>, paramNames [Param1, Wiggle_Frequency, Wiggle_Amplitude_1, Wiggle_Amplitude_2]}
  MaterialFunctionCall("FadeSensitivity (optional)"=ScalarParameter() {defaultValue=2, parameterName=Fade edge}) {materialFunction=&FadeBasedOnViewAngle}
  n26 = ParticleColor()
  MaterialFunctionCall(Emissive=Multiply(A=@n17, B=@n26.RGB)) {materialFunction=&MF_FX_EmissiveAdjustments}
  n27 = ComponentMask(None=TextureCoordinate()) {r=true}
  n28 = Multiply(A=@n27) {constB=5}
  n29 = Multiply(A=OneMinus(None=@n27)) {constB=5}
  n30 = MaterialFunctionCall(
   Brightness=ScalarParameter() {defaultValue=3, group=Alpha, parameterName="CameraNormal_Fresnel_Brightness (S)"}
   Exponent=ScalarParameter() {defaultValue=5, group=Alpha, parameterName="CameraNormal_Fresnel_Exponent (S)"}
   Vertex / Pixel Normal=VertexNormalWS()
  ) {materialFunction=&MF_Fresnel_FX}
  n31 = Constant() {r=1}
  Multiply(
   A=Saturate(None=Multiply(A=@n28, B=@n29))
   B=Multiply(
    A=Multiply(A=@n17, B=@n26.A)
    B=StaticSwitchParameter(True=@n30.Fresnel Double Sided, False=@n31) {defaultValue=true, parameterName=Side_Fade}
   )
  )
 ]
}
