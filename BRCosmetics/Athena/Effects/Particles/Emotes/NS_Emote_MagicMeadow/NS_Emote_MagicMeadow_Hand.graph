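# Niagara system for the hand effect of the Magic Meadow emote, faking a warm glow with a sprite.
# A single FakeLight emitter burst-spawns a long-lived, bright HDR orange sprite, nudged by a camera offset.
# A curve on alpha fades it in quickly, holds it, then fades it out near the end of its life.
# User parameters for front-end preview, player speed and skeletal mesh are declared but not referenced by the modules shown.
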NiagaraSystem /BRCosmetics/Athena/Effects/Particles/Emotes/NS_Emote_MagicMeadow/NS_Emote_MagicMeadow_Hand
dependencies {usedDataInterfaces [&NiagaraDataInterfaceCurve], usedDynamicInputs [], usedModules [&SystemState]}
emitters {FakeLight {
  emitterUpdateScript [
   &EmitterState()
   &SpawnBurst_Instantaneous(Use Spawn Probability=false, UseLoopCountLimit=false)
  ]
  particleSpawnScript [
   &InitializeParticle(
    Color=<rgba 100 27.6164 5.5122 2>
    Lifetime=9.8
    Sprite Size Mode=::NewEnumerator3
    Uniform Sprite Size=100
    UsePositionOffset=false
    Write Material Random=true
   )
   &CameraOffset(Camera Offset Amount=5)
  ]
  particleUpdateScript [
   &ParticleState()
   &ScaleColor(
    Scale Alpha=&FloatFromCurve(FloatCurve=DataInterface() {propertyValues {Curve {keys [<at 0 0>,<at 0.05 1>,<at 0.9 1>,<at 0.96 0>]}}})
    ScaleA=true
    ScaleRGB=true
   )
  ]
  rendererClasses [&NiagaraSpriteRendererProperties]
  renderers [{rendererClass=&NiagaraSpriteRendererProperties}]
 }}
summary {
 systemName=NS_Emote_MagicMeadow_Hand
 emitters [{emitterName=FakeLight, rendererClasses [&NiagaraSpriteRendererProperties]}]
 userVariables [
  User.bIsFrontEndPreview {type=NiagaraBool}
  User.PlayerSpeed {type=NiagaraFloat, defaultValue=0, description=""}
  User.SkeletalMesh {
   type=NiagaraDataInterfaceSkeletalMesh
   defaultValue {
    struct=NiagaraExt_VariableValue_DataInterface
    value {dataInterface=&self:NiagaraDataInterfaceSkeletalMesh_0}
   }
  }
 ]
}
