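# Niagara system for the MagicMeadow emote: a glowing, fiery "Igni" symbol that appears with flames and sparks, then burns away.
# Igni_Main and Igni_SideMesh last for System.FX_Duration, with curves driving color fades, material parameters, glow and a flickering light.
# Forward sparks and soft flames fire at the start, while Igni_Burn and Embers_Burst are delayed relative to that duration to form the finish.
# Some older spark and ember emitters remain disabled.
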NiagaraSystem /BRCosmetics/Athena/Effects/Particles/Emotes/NS_Emote_MagicMeadow/NS_Emote_MagicMeadow_Main
dependencies {
 usedDataInterfaces [
  &NiagaraDataInterfaceCurve
  &NiagaraDataInterfaceColorCurve
  &NiagaraDataInterfaceVectorCurve
  &NiagaraDataInterfaceVector2DCurve
  &NiagaraDataInterfaceSpriteRendererInfo
 ]
 usedDynamicInputs []
 usedModules [
  &self:SystemSpawnScript.SystemScriptSource.SystemScriptGraph.NiagaraNodeAssignment_1.SetVariables_#1
  &self:SystemSpawnScript.SystemScriptSource.SystemScriptGraph.NiagaraNodeAssignment_2.SetVariables_#2
  &SystemState
 ]
}
emitters {
 Embers_Burst {
  emitterUpdateScript [
   &EmitterState(
    Life Cycle Mode=::NewEnumerator1
    Loop Behavior=::NewEnumerator1
    Loop Delay=&Add_Float(A=$System.FX_Duration, B=-0.67)
    Loop Duration=2
    UseLoopDelay=true
   )
   &SpawnBurst_Instantaneous(Spawn Count=20, Use Spawn Probability=false, UseLoopCountLimit=false)
   &SpawnRate(
    SpawnRate=&FloatFromCurve(
     FloatCurve=DataInterface() {propertyValues {Curve {keys [<at 0 1>,<at 0.666 0.333>,<at 3 0>]}}}
     CurveIndex=$Emitter.LoopedAge
     Scale Curve=111
    )
    Use Spawn Probability=false
   ) {enabled=false}
  ]
  particleSpawnScript [
   /SetVariables(Particles.PerParticleRandom:NiagaraFloat=&RandomRangeFloat(Minimum=0, Maximum=1)) {isSetParameters=true, module=SetVariables_#3}
   &InitializeParticle(
    Color Mode=::NewEnumerator0
    Lifetime=&Lerp_Float(A=0.2, B=0.65, Alpha=$Particles.PerParticleRandom)
    Sprite Size=&Multiply_Vector2D(
     A=<0.5 1>
     B=&MakeVector2D(
      X=&FloatFromCurve(
       FloatCurve=DataInterface() {propertyValues {Curve {keys [<at 0 1>,<at 0.8 2 RCIM_Constant>,<at 0.95 3 RCIM_Constant>]}}}
       CurveIndex=$Particles.PerParticleRandom
      )
      Y=&FloatFromCurve(
       FloatCurve=DataInterface() {propertyValues {Curve {keys [<at 0 1>,<at 0.8 2 RCIM_Constant>,<at 0.95 3>]}}}
       CurveIndex=$Particles.PerParticleRandom
      )
     )
    )
    Sprite Size Mode=::NewEnumerator4
    UsePositionOffset=false
    Write Material Random=true
   )
   &AddVelocity(
    Cone Angle=33
    Cone Axis=<0 0.25 1>
    Module.UseConeFalloff=false
    Module.UseVelDistribution=false
    Override Seed=false
    Velocity Mode=::NewEnumerator2
    Velocity Speed=&FloatFromCurve(
     FloatCurve=DataInterface() {propertyValues {Curve {keys [<at 0 0.5>,<at 0.333 1>,<at 0.666 0.1>,<at 3 0>]}}}
     CurveIndex=$Emitter.LoopedAge
     Scale Curve=&RandomRangeFloat(Minimum=444, Maximum=999)
    )
   ) {enabled=false}
   &AddVelocity(Module.Use Falloff Distance=false, Override Seed=false, Velocity Mode=::NewEnumerator1, Velocity Speed=400) {module=AddVelocity001}
   &ShapeLocation(MaskSpawnGroup=false, Offset Mode=::NewEnumerator0, Override Seed=false, Sphere Radius=16, UseOwnerScale=true)
  ]
  particleUpdateScript [
   &ParticleState()
   &AccelerationForce()
   &Drag(
    Drag=&FloatFromCurve(
     FloatCurve=DataInterface() {propertyValues {Curve {keys [<at 0 0.5>,<at 0.5 0.5>,<at 1 1>]}}}
     Scale Curve=&RandomRangeFloat(Minimum=6, Maximum=22)
    )
    Use Linear Drag=true
    Use Rotational Drag=false
   )
   /Niagara/Modules/Update/Forces/CurlNoiseForce(
    Noise Frequency=0.03
    Noise Strength=&FloatFromCurve(
     FloatCurve=DataInterface() {propertyValues {Curve {keys [{tangentMode=RCTM_User, time=0, value=0}, {tangentMode=RCTM_User, time=1, value=1}]}}}
     Scale Curve=&Lerp_Float(A=4444, B=444, Alpha=$Particles.PerParticleRandom)
    )
    Pan Noise=true
    Pan Noise Field=<0 0 1>
   )
   &SolveForcesAndVelocity(Clamp Velocity=false, Limit Acceleration=false)
   &ScaleColor(
    Scale RGB=&MakeVectorFromLinearColorRGB(Color=&ColorFromCurve(ColorCurve=DataInterface() {propertyValues {
        AlphaCurve {keys [<at 0 1 RCIM_Cubic>]}
        BlueCurve {keys [<at 0 0 RCIM_Cubic>,<at 1 0 RCIM_Cubic>]}
        GreenCurve {keys [<at 0 9.1667 RCIM_Cubic>,<at 1 0 RCIM_Cubic>]}
        RedCurve {keys [<at 0 50 RCIM_Cubic>,<at 1 1 RCIM_Cubic>]}
       }}))
    ScaleA=true
    ScaleRGB=true
   )
   &ScaleSpriteSizeBySpeed(
    Max Scale Factor=<1.666 13>
    Min Scale Factor=<1.333 1.333>
    Sample Scale Factor By Curve=true
    Scale Factor Curve=DataInterface() {propertyValues {Curve {keys [<at 0 0>, {arriveTangent=1.4989, interpMode=RCIM_Cubic, tangentMode=RCTM_Break, time=1, value=1}]}}}
    Velocity Threshold=999
   )
   &ScaleSpriteSize(Uniform Curve Sprite Scale=DataInterface() {propertyValues {Curve {keys [
        {tangentMode=RCTM_User, time=0, value=1}
        {arriveTangent=-0.0405, interpMode=RCIM_Cubic, leaveTangent=-0.0405, tangentMode=RCTM_User, time=0.7009, value=1}
        {arriveTangent=-8.0089, interpMode=RCIM_Cubic, leaveTangent=-8.0089, tangentMode=RCTM_User, time=1, value=0}
       ]}}})
  ]
  rendererClasses [&NiagaraSpriteRendererProperties]
  renderers [{rendererClass=&NiagaraSpriteRendererProperties}]
 }
 Embers_Forward002 {
  bEnabled=false
  emitterUpdateScript [
   &EmitterState(
    Inactive Response=::NewEnumerator1
    Life Cycle Mode=::NewEnumerator1
    Loop Behavior=::NewEnumerator1
    Loop Delay=0.1
    Loop Duration=$System.FX_Duration
    UseLoopDelay=true
   )
   &SpawnRate(
    SpawnRate=&FloatFromCurve(
     FloatCurve=DataInterface() {propertyValues {Curve {keys [
         {arriveTangent=-2.6167, interpMode=RCIM_Cubic, leaveTangent=-2.6167, tangentMode=RCTM_User, time=0.5, value=1}
         {arriveTangent=-0.1055, interpMode=RCIM_Cubic, leaveTangent=-0.1055, tangentMode=RCTM_User, time=0.9, value=0}
        ]}}}
     CurveIndex=$Emitter.NormalizedLoopAge
     Scale Curve=35
    )
    Use Spawn Probability=false
   )
  ]
  particleSpawnScript [
   /SetVariables(Particles.PerParticleRandom:NiagaraFloat=&RandomRangeFloat(Minimum=0, Maximum=1)) {isSetParameters=true, module=SetVariables_#3}
   &InitializeParticle(
    Color Maximum=<rgba 1 0.4712 0.059 1>
    Color Minimum=<rgba 1 0.216 0.0459 1>
    Color Mode=::NewEnumerator2
    Lifetime=&Lerp_Float(A=0.2, B=0.7, Alpha=$Particles.PerParticleRandom)
    Sprite Size=&Multiply_Vector2D_ByFloat(
     Float=&RandomRangeFloat(Minimum=30, Maximum=45)
     Vector2D=&RandomRangeVector2D(Minimum=<0.3 0.8>, Maximum=<0.6 1>)
    )
    Sprite Size Mode=::NewEnumerator4
    UsePositionOffset=false
    Write Material Random=true
   )
   &AddVelocity(
    Cone Angle=33
    Cone Axis=<0 0.25 1>
    Module.UseConeFalloff=false
    Module.UseVelDistribution=false
    Override Seed=false
    Velocity Mode=::NewEnumerator2
    Velocity Speed=&FloatFromCurve(
     FloatCurve=DataInterface() {propertyValues {Curve {keys [<at 0 0.5>,<at 0.333 1>,<at 0.666 0.1>,<at 3 0>]}}}
     CurveIndex=$Emitter.LoopedAge
     Scale Curve=&RandomRangeFloat(Minimum=444, Maximum=999)
    )
   ) {enabled=false}
   &ShapeLocation(MaskSpawnGroup=false, Offset Mode=::NewEnumerator0, Override Seed=false, Sphere Radius=1, UseOwnerScale=true)
   &AddVelocity(
    Cone Angle=60
    Cone Axis=<-1 0 0>
    Module.UseConeFalloff=false
    Module.UseVelDistribution=false
    Override Seed=false
    Rotation Angle Type:ENiagara_AngleInput=::NewEnumerator1
    Rotation Mode=::NewEnumerator2
    Velocity Mode=::NewEnumerator2
    Velocity Speed=&RandomRangeFloat(Minimum=30, Maximum=60)
    Yaw / Pitch / Roll:Vector3f=<-0.25 0 0>
   ) {module=AddVelocity001}
   &SubUVAnimation(
    "Sprite Renderer "=DataInterface() {propertyValues {SpriteRenderer=&self:Embers_Forward002_3.NiagaraSpriteRendererProperties_2}}
    SubUV Animation Mode:ENiagara_SubUVLookupModeV2=::NewEnumerator2
    UseEndFrame=false
    UseStartFrame=false
   )
  ]
  particleUpdateScript [
   &ParticleState()
   &AccelerationForce() {enabled=false}
   &Drag(
    Drag=&FloatFromCurve(
     FloatCurve=DataInterface() {propertyValues {Curve {keys [<at 0 0.5>,<at 0.3 0.5>,<at 1 1>]}}}
     Scale Curve=&RandomRangeFloat(Minimum=10, Maximum=15)
    )
    Use Linear Drag=true
    Use Rotational Drag=false
   )
   &CurlNoiseForce(
    Noise Frequency=0.1
    Noise Strength=&FloatFromCurve(
     FloatCurve=DataInterface() {propertyValues {Curve {keys [{tangentMode=RCTM_User, time=0, value=0}, {tangentMode=RCTM_User, time=1, value=1}]}}}
     Scale Curve=&Lerp_Float(A=500, B=1200, Alpha=$Particles.PerParticleRandom)
    )
    Pan Noise=true
    Pan Noise Field=<0 0 1>
   ) {enabled=false}
   &SolveForcesAndVelocity(Clamp Velocity=false, Limit Acceleration=false)
   &ScaleColor(
    Scale RGB=&MakeVectorFromLinearColorRGB(Color=&ColorFromCurve(ColorCurve=DataInterface() {propertyValues {
        AlphaCurve {keys [<at 0 1 RCIM_Cubic>]}
        BlueCurve {keys [<at 0 0>,<at 0.1089 0 RCIM_Cubic>,<at 1 0 RCIM_Cubic>]}
        GreenCurve {keys [<at 0 0>,<at 0.1089 0.1833 RCIM_Cubic>,<at 1 0 RCIM_Cubic>]}
        RedCurve {keys [<at 0 0.03>, {arriveTangent=-0.02, interpMode=RCIM_Cubic, leaveTangent=-0.02, time=0.1089, value=1}, <at 1 0.01 RCIM_Cubic>]}
       }}))
    ScaleA=true
    ScaleRGB=true
   ) {enabled=false}
   &ScaleColor(
    Scale Alpha=&FloatFromCurve(FloatCurve=DataInterface() {propertyValues {Curve {keys [
         {arriveTangent=4.0017, interpMode=RCIM_Cubic, leaveTangent=4.0017, tangentMode=RCTM_User, time=0, value=-0.7}
         {arriveTangent=0.7858, interpMode=RCIM_Cubic, leaveTangent=0.7858, tangentMode=RCTM_User, time=1, value=1}
        ]}}})
    ScaleA=true
    ScaleRGB=true
   ) {module=ScaleColor001}
   &ScaleSpriteSizeBySpeed(
    Max Scale Factor=<1.666 13>
    Min Scale Factor=<1.333 1.333>
    Sample Scale Factor By Curve=true
    Scale Factor Curve=DataInterface() {propertyValues {Curve {keys [<at 0 0>, {arriveTangent=1.4989, interpMode=RCIM_Cubic, tangentMode=RCTM_Break, time=1, value=1}]}}}
    Velocity Threshold=999
   ) {enabled=false}
   &ScaleSpriteSize(
    Non-Uniform Curve Sprite Scale=DataInterface() {propertyValues {
      XCurve {keys [{tangentMode=RCTM_User, time=0, value=1}, {tangentMode=RCTM_User, time=1, value=0}]}
      YCurve {keys [{tangentMode=RCTM_User, time=0, value=1}, {tangentMode=RCTM_User, time=1, value=1}]}
     }}
    Scale Sprite Size Mode=::NewEnumerator3
   ) {enabled=false}
   &DynamicMaterialParameters(
    Index 0 Param 1=&FloatFromCurve(FloatCurve=DataInterface() {propertyValues {Curve {keys [
         <at 0 0 RCIM_Cubic>
         {arriveTangent=-0.1623, interpMode=RCIM_Cubic, leaveTangent=-0.1623, tangentMode=RCTM_User, time=0.1, value=1}
         <at 1 0 RCIM_Cubic>
        ]}}})
    Module.Param0WriteEnabled=true
    Module.Param1WriteEnabled=true
    Module.Param2WriteEnabled=true
    Module.Param3WriteEnabled=true
   )
  ]
  rendererClasses [&NiagaraSpriteRendererProperties]
  renderers [{rendererClass=&NiagaraSpriteRendererProperties}]
 }
 Flames {
  emitterUpdateScript [
   &EmitterState(
    Inactive Response=::NewEnumerator1
    Life Cycle Mode=::NewEnumerator1
    Loop Behavior=::NewEnumerator1
    Loop Delay=0.16
    Loop Duration=&Add_Float(A=$System.FX_Duration, B=-0.7)
    UseLoopDelay=true
   )
   &SpawnRate(
    SpawnRate=&Multiply_Float(
     A=8
     B=&FloatFromCurve(
      FloatCurve=DataInterface() {propertyValues {Curve {keys [<at 0.6 1 RCIM_Cubic>, {arriveTangent=-2.841, interpMode=RCIM_Cubic, leaveTangent=-2.841, tangentMode=RCTM_User, time=1, value=0}]}}}
      CurveIndex=$Emitter.NormalizedLoopAge
     )
    )
    Use Spawn Probability=false
   )
   &SpawnBurst_Instantaneous(Use Spawn Probability=false, UseLoopCountLimit=false) {enabled=false}
  ]
  particleSpawnScript [
   &InitializeParticle(
    Color=&RandomRangeLinearColor(Minimum=<rgba 2 0.86 0.16 1>, Maximum=<rgba 2 0.4774 0 0.8>)
    Lifetime=&RandomRangeFloat(Minimum=0.7, Maximum=1.2)
    Position Offset=&RandomRangeVector(Minimum=<0 -1 0>, Maximum=<0 1 0>)
    Sprite Size=&Multiply_Vector2D_ByFloat(Float=0.66, Vector2D=&RandomRangeVector2D(Minimum=<65 65>, Maximum=<75 75>))
    Sprite Size Mode=::NewEnumerator4
    UsePositionOffset=true
    Write Material Random=true
   )
   &SubUVAnimation(End Frame=4, SubUV Animation Mode:ENiagara_SubUVLookupMode=::NewEnumerator2)
   &AddVelocity(Velocity=&RandomRangeVector(Minimum=<0 0 10>, Maximum=<0 0 23>))
   &SpriteFacingAndAlignment(Sprite Alignment=<0 0 1>, Sprite Facing=<0 1 0>, WriteAlignment=true, WriteFacing=true)
   &DynamicMaterialParameters(
    Index 0 Param 2=$System.WPO_Wiggle_Freq1
    Index 0 Param 3=$System.WPO_Wiggle_Amp1
    Index 0 Param 4=$System.WPO_Wiggle_Amp2
    Module.Param0WriteEnabled=false
    Module.Param1WriteEnabled=true
    Module.Param2WriteEnabled=true
    Module.Param3WriteEnabled=true
   ) {enabled=false, module=DynamicMaterialParameters001}
  ]
  particleUpdateScript [
   &ParticleState()
   &ScaleSpriteSize(
    Non-Uniform Curve Sprite Scale=DataInterface() {propertyValues {
      XCurve {keys [{arriveTangent=0.4783, interpMode=RCIM_Cubic, leaveTangent=0.4783, tangentMode=RCTM_User, time=0, value=0.85}, <at 1 1 RCIM_Cubic>]}
      YCurve {keys [{arriveTangent=1.0983, interpMode=RCIM_Cubic, leaveTangent=1.0983, tangentMode=RCTM_User, time=0, value=0.75}, <at 1 1.1 RCIM_Cubic>]}
     }}
    Scale Sprite Size Mode=::NewEnumerator3
   )
   &Drag(Drag=1.7, Use Linear Drag=true, Use Rotational Drag=true)
   &SolveForcesAndVelocity(Clamp Velocity=false, Limit Acceleration=false)
   &ScaleColor(
    Scale Alpha=&RandomRangeFloat(
     Minimum=&FloatFromCurve(FloatCurve=DataInterface() {propertyValues {Curve {keys [
          <at 0 1 RCIM_Cubic>
          <at 0.1 0.2 RCIM_Cubic>
          {arriveTangent=0.9616, interpMode=RCIM_Cubic, leaveTangent=0.9616, tangentMode=RCTM_User, time=1, value=1}
         ]}}})
     Maximum=&FloatFromCurve(FloatCurve=DataInterface() {propertyValues {Curve {keys [
          <at 0 1 RCIM_Cubic>
          <at 0.1 -1 RCIM_Cubic>
          {arriveTangent=3.0424, interpMode=RCIM_Cubic, leaveTangent=3.0424, tangentMode=RCTM_User, time=1, value=1}
         ]}}})
    )
    ScaleA=true
    ScaleRGB=true
   )
   &DynamicMaterialParameters(
    Index 0 Param 1=&FloatFromCurve(
     FloatCurve=DataInterface() {propertyValues {Curve {keys [
         <at 0 0 RCIM_Cubic>
         {arriveTangent=-1.0044, interpMode=RCIM_Cubic, leaveTangent=-1.0044, tangentMode=RCTM_User, time=0.25, value=1}
         <at 1 0 RCIM_Cubic>
        ]}}}
     Scale Curve=0.05
    )
    Module.Param0WriteEnabled=true
    Module.Param1WriteEnabled=true
    Module.Param2WriteEnabled=true
    Module.Param3WriteEnabled=true
   )
  ]
  rendererClasses [&NiagaraSpriteRendererProperties]
  renderers [{rendererClass=&NiagaraSpriteRendererProperties}]
 }
 Forward_Flames_Soft {
  emitterUpdateScript [
   &EmitterState(
    Inactive Response=::NewEnumerator1
    Life Cycle Mode=::NewEnumerator1
    Loop Behavior=::NewEnumerator1
    Loop Duration=$System.FX_Duration
    UseLoopDelay=false
   )
   &SpawnBurst_Instantaneous(Spawn Count=10, Spawn Time=0.028, Use Spawn Probability=false, UseLoopCountLimit=false)
  ]
  particleSpawnScript [
   /SetVariables(Particles.PerParticleRandom:NiagaraFloat=&RandomRangeFloat(Minimum=0, Maximum=1)) {isSetParameters=true, module=SetVariables_#3}
   &InitializeParticle(
    Color=<rgba 4 1.1253 0.08 1>
    Lifetime=&Lerp_Float(A=0.25, B=0.4, Alpha=$Particles.PerParticleRandom)
    Sprite Size=&Multiply_Vector2D_ByFloat(Float=40, Vector2D=&RandomRangeVector2D(Minimum=<0.06 0.6>, Maximum=<0.25 1>))
    Sprite Size Mode=::NewEnumerator4
    UsePositionOffset=false
    Write Material Random=true
   )
   &AddVelocity(
    Cone Angle=33
    Cone Axis=<0 0.25 1>
    Module.UseConeFalloff=false
    Module.UseVelDistribution=false
    Override Seed=false
    Velocity Mode=::NewEnumerator2
    Velocity Speed=&FloatFromCurve(
     FloatCurve=DataInterface() {propertyValues {Curve {keys [<at 0 0.5>,<at 0.333 1>,<at 0.666 0.1>,<at 3 0>]}}}
     CurveIndex=$Emitter.LoopedAge
     Scale Curve=&RandomRangeFloat(Minimum=444, Maximum=999)
    )
   ) {enabled=false}
   &ShapeLocation(MaskSpawnGroup=false, Offset=<0 -6 0>, Offset Mode=::NewEnumerator0, Override Seed=false, Sphere Radius=1, UseOwnerScale=true)
   &AddVelocity(
    Module.Use Falloff Distance=false
    Origin Offset=<0 -8 0>
    Override Seed=false
    Velocity Mode=::NewEnumerator1
    Velocity Speed=&RandomRangeFloat(Minimum=100, Maximum=200)
   ) {module=AddVelocity001}
   &SubUVAnimation(
    "Sprite Renderer "=DataInterface() {propertyValues {SpriteRenderer=&self:Forward_Flames_Soft_1.NiagaraSpriteRendererProperties_2}}
    SubUV Animation Mode:ENiagara_SubUVLookupModeV2=::NewEnumerator2
    UseEndFrame=false
    UseStartFrame=false
   )
  ]
  particleUpdateScript [
   &ParticleState()
   &AccelerationForce() {enabled=false}
   &Drag(
    Drag=&FloatFromCurve(
     FloatCurve=DataInterface() {propertyValues {Curve {keys [<at 0 0.5>,<at 0.3 0.5>,<at 1 1>]}}}
     Scale Curve=&RandomRangeFloat(Minimum=7, Maximum=12)
    )
    Use Linear Drag=true
    Use Rotational Drag=false
   )
   &CurlNoiseForce(
    Noise Frequency=0.1
    Noise Strength=&FloatFromCurve(
     FloatCurve=DataInterface() {propertyValues {Curve {keys [{tangentMode=RCTM_User, time=0, value=0}, {tangentMode=RCTM_User, time=1, value=1}]}}}
     Scale Curve=&Lerp_Float(A=500, B=1200, Alpha=$Particles.PerParticleRandom)
    )
    Pan Noise=true
    Pan Noise Field=<0 0 1>
   ) {enabled=false}
   &SolveForcesAndVelocity(Clamp Velocity=false, Limit Acceleration=false)
   &ScaleSpriteSizeBySpeed(Max Scale Factor=<1 2>) {module=ScaleSpriteSizeBySpeed001}
   &ScaleColor(
    Scale RGB=&MakeVectorFromLinearColorRGB(Color=&ColorFromCurve(ColorCurve=DataInterface() {propertyValues {
        AlphaCurve {keys [<at 0 1 RCIM_Cubic>]}
        BlueCurve {keys [<at 0 0>,<at 0.1089 0 RCIM_Cubic>,<at 1 0 RCIM_Cubic>]}
        GreenCurve {keys [<at 0 0>,<at 0.1089 0.1833 RCIM_Cubic>,<at 1 0 RCIM_Cubic>]}
        RedCurve {keys [<at 0 0.03>, {arriveTangent=-0.02, interpMode=RCIM_Cubic, leaveTangent=-0.02, time=0.1089, value=1}, <at 1 0.01 RCIM_Cubic>]}
       }}))
    ScaleA=true
    ScaleRGB=true
   ) {enabled=false}
   &ScaleColor(
    Scale Alpha=&FloatFromCurve(FloatCurve=DataInterface() {propertyValues {Curve {keys [
         {arriveTangent=2.2951, interpMode=RCIM_Cubic, leaveTangent=2.2951, tangentMode=RCTM_User, time=0, value=-0.7}
         {interpMode=RCIM_Cubic, tangentMode=RCTM_User, time=1, value=1}
        ]}}})
    ScaleA=true
    ScaleRGB=true
   ) {module=ScaleColor001}
   &ScaleSpriteSizeBySpeed(
    Max Scale Factor=<1.666 13>
    Min Scale Factor=<1.333 1.333>
    Sample Scale Factor By Curve=true
    Scale Factor Curve=DataInterface() {propertyValues {Curve {keys [<at 0 0>, {arriveTangent=1.4989, interpMode=RCIM_Cubic, tangentMode=RCTM_Break, time=1, value=1}]}}}
    Velocity Threshold=999
   ) {enabled=false}
   &ScaleSpriteSize(Uniform Curve Sprite Scale=DataInterface() {propertyValues {Curve {keys [
        {tangentMode=RCTM_User, time=0, value=1}
        {arriveTangent=-0.0405, interpMode=RCIM_Cubic, leaveTangent=-0.0405, tangentMode=RCTM_User, time=0.7009, value=1}
        {arriveTangent=-8.0089, interpMode=RCIM_Cubic, leaveTangent=-8.0089, tangentMode=RCTM_User, time=1, value=0}
       ]}}}) {enabled=false}
   &DynamicMaterialParameters(
    Index 0 Param 1=&FloatFromCurve(
     FloatCurve=DataInterface() {propertyValues {Curve {keys [{arriveTangent=-0.5499, interpMode=RCIM_Cubic, leaveTangent=-0.5499, tangentMode=RCTM_User, time=0.4, value=1}, <at 1 0 RCIM_Cubic>]}}}
     Scale Curve=2
    )
    Module.Param0WriteEnabled=true
    Module.Param1WriteEnabled=true
    Module.Param2WriteEnabled=true
    Module.Param3WriteEnabled=true
   )
  ]
  rendererClasses [&NiagaraSpriteRendererProperties]
  renderers [{rendererClass=&NiagaraSpriteRendererProperties}]
 }
 Forward_Sparks {
  bEnabled=false
  emitterUpdateScript [
   &EmitterState(
    Inactive Response=::NewEnumerator1
    Life Cycle Mode=::NewEnumerator1
    Loop Behavior=::NewEnumerator1
    Loop Duration=$System.FX_Duration
    UseLoopDelay=false
   )
   &SpawnRate(SpawnRate=200, Use Spawn Probability=false) {enabled=false, module=SpawnRate001}
   &SpawnBurst_Instantaneous(Spawn Count=44, Use Spawn Probability=false, UseLoopCountLimit=false) {enabled=false}
   &SpawnRate(
    SpawnRate=&FloatFromCurve(
     FloatCurve=DataInterface() {propertyValues {Curve {keys [<at 0 1>,<at 0.7 0.4>,<at 0.92 0>]}}}
     CurveIndex=$Emitter.NormalizedLoopAge
     Scale Curve=110
    )
    Use Spawn Probability=false
   )
  ]
  particleSpawnScript [
   /SetVariables(Particles.PerParticleRandom:NiagaraFloat=&RandomRangeFloat(Minimum=0, Maximum=1)) {isSetParameters=true, module=SetVariables_#3}
   &InitializeParticle(
    Color Mode=::NewEnumerator0
    Lifetime=&Lerp_Float(A=0.2, B=0.65, Alpha=$Particles.PerParticleRandom)
    Sprite Size=&Multiply_Vector2D(
     A=<0.6 1.2>
     B=&MakeVector2D(
      X=&FloatFromCurve(
       FloatCurve=DataInterface() {propertyValues {Curve {keys [<at 0 1>,<at 0.8 2 RCIM_Constant>,<at 0.95 3 RCIM_Constant>]}}}
       CurveIndex=$Particles.PerParticleRandom
      )
      Y=&FloatFromCurve(
       FloatCurve=DataInterface() {propertyValues {Curve {keys [<at 0 1>,<at 0.8 2 RCIM_Constant>,<at 0.95 3>]}}}
       CurveIndex=$Particles.PerParticleRandom
      )
     )
    )
    Sprite Size Mode=::NewEnumerator4
    UsePositionOffset=false
    Write Material Random=true
   )
   &AddVelocity(
    Cone Angle=33
    Cone Axis=<0 0.25 1>
    Module.UseConeFalloff=false
    Module.UseVelDistribution=false
    Override Seed=false
    Velocity Mode=::NewEnumerator2
    Velocity Speed=&FloatFromCurve(
     FloatCurve=DataInterface() {propertyValues {Curve {keys [<at 0 0.5>,<at 0.333 1>,<at 0.666 0.1>,<at 3 0>]}}}
     CurveIndex=$Emitter.LoopedAge
     Scale Curve=&RandomRangeFloat(Minimum=444, Maximum=999)
    )
   ) {enabled=false}
   &ShapeLocation(MaskSpawnGroup=false, Offset Mode=::NewEnumerator0, Override Seed=false, Sphere Radius=8, UseOwnerScale=true)
   &AddVelocity(
    Cone Axis=<-1 0 0>
    Module.UseConeFalloff=false
    Module.UseVelDistribution=false
    Override Seed=false
    Rotation Angle Type:ENiagara_AngleInput=::NewEnumerator1
    Rotation Mode=::NewEnumerator2
    Velocity Mode=::NewEnumerator2
    Velocity Speed=&RandomRangeFloat(Minimum=400, Maximum=700)
    Yaw / Pitch / Roll:Vector3f=<-0.25 0 0>
   ) {enabled=false, module=AddVelocity001}
   &AddVelocity(
    Module.Use Falloff Distance=false
    Origin Offset=<0 -30 0>
    Override Seed=false
    Velocity Mode=::NewEnumerator1
    Velocity Speed=&RandomRangeFloat(Minimum=400, Maximum=700)
   ) {module=AddVelocity002}
  ]
  particleUpdateScript [
   &ParticleState()
   &AccelerationForce() {enabled=false}
   &Drag(
    Drag=&FloatFromCurve(
     FloatCurve=DataInterface() {propertyValues {Curve {keys [<at 0 0.5>,<at 0.5 0.5>,<at 1 1>]}}}
     Scale Curve=&RandomRangeFloat(Minimum=10, Maximum=30)
    )
    Use Linear Drag=true
    Use Rotational Drag=false
   )
   &CurlNoiseForce(
    Noise Frequency=0.1
    Noise Strength=&FloatFromCurve(
     FloatCurve=DataInterface() {propertyValues {Curve {keys [{tangentMode=RCTM_User, time=0, value=0}, {tangentMode=RCTM_User, time=1, value=1}]}}}
     Scale Curve=&Lerp_Float(A=500, B=1200, Alpha=$Particles.PerParticleRandom)
    )
    Pan Noise=true
    Pan Noise Field=<0 0 1>
   )
   &SolveForcesAndVelocity(Clamp Velocity=false, Limit Acceleration=false)
   &ScaleColor(
    Scale RGB=&MakeVectorFromLinearColorRGB(Color=&ColorFromCurve(ColorCurve=DataInterface() {propertyValues {
        AlphaCurve {keys [<at 0 1 RCIM_Cubic>]}
        BlueCurve {keys [<at 0.0265 0>,<at 0.1089 0 RCIM_Cubic>,<at 1 0 RCIM_Cubic>]}
        GreenCurve {keys [<at 0.0265 0>,<at 0.1089 1.8333 RCIM_Cubic>,<at 1 0 RCIM_Cubic>]}
        RedCurve {keys [<at 0.0265 0.2>, {arriveTangent=0.8, interpMode=RCIM_Cubic, time=0.1089, value=10}, <at 1 1 RCIM_Cubic>]}
       }}))
    ScaleA=true
    ScaleRGB=true
   )
   &ScaleSpriteSizeBySpeed(
    Max Scale Factor=<1.666 13>
    Min Scale Factor=<1.333 1.333>
    Sample Scale Factor By Curve=true
    Scale Factor Curve=DataInterface() {propertyValues {Curve {keys [<at 0 0>, {arriveTangent=1.4989, interpMode=RCIM_Cubic, tangentMode=RCTM_Break, time=1, value=1}]}}}
    Velocity Threshold=999
   )
   &ScaleSpriteSize(Uniform Curve Sprite Scale=DataInterface() {propertyValues {Curve {keys [
        {tangentMode=RCTM_User, time=0, value=1}
        {arriveTangent=-0.0405, interpMode=RCIM_Cubic, leaveTangent=-0.0405, tangentMode=RCTM_User, time=0.7009, value=1}
        {arriveTangent=-8.0089, interpMode=RCIM_Cubic, leaveTangent=-8.0089, tangentMode=RCTM_User, time=1, value=0}
       ]}}})
  ]
  rendererClasses [&NiagaraSpriteRendererProperties]
  renderers [{rendererClass=&NiagaraSpriteRendererProperties}]
 }
 Forward_Sparks_New {
  emitterUpdateScript [
   &EmitterState(
    Inactive Response=::NewEnumerator1
    Life Cycle Mode=::NewEnumerator1
    Loop Behavior=::NewEnumerator1
    Loop Duration=$System.FX_Duration
    UseLoopDelay=false
   )
   &SpawnRate(SpawnRate=200, Use Spawn Probability=false) {enabled=false, module=SpawnRate001}
   &SpawnBurst_Instantaneous(Spawn Count=44, Use Spawn Probability=false, UseLoopCountLimit=false) {enabled=false}
   &SpawnRate(
    SpawnRate=&FloatFromCurve(
     FloatCurve=DataInterface() {propertyValues {Curve {keys [<at 0 1>,<at 0.75 0.55>,<at 0.92 0>]}}}
     CurveIndex=$Emitter.NormalizedLoopAge
     Scale Curve=110
    )
    Use Spawn Probability=false
   )
  ]
  particleSpawnScript [
   /SetVariables(Particles.PerParticleRandom:NiagaraFloat=&RandomRangeFloat(Minimum=0, Maximum=1)) {isSetParameters=true, module=SetVariables_#3}
   &InitializeParticle(
    Color Mode=::NewEnumerator0
    Lifetime=&Lerp_Float(A=0.2, B=0.65, Alpha=$Particles.PerParticleRandom)
    Sprite Size=&Multiply_Vector2D(
     A=<0.6 1.2>
     B=&MakeVector2D(
      X=&FloatFromCurve(
       FloatCurve=DataInterface() {propertyValues {Curve {keys [<at 0 1>,<at 0.8 2 RCIM_Constant>,<at 0.95 3 RCIM_Constant>]}}}
       CurveIndex=$Particles.PerParticleRandom
      )
      Y=&FloatFromCurve(
       FloatCurve=DataInterface() {propertyValues {Curve {keys [<at 0 1>,<at 0.8 2 RCIM_Constant>,<at 0.95 3>]}}}
       CurveIndex=$Particles.PerParticleRandom
      )
     )
    )
    Sprite Size Mode=::NewEnumerator4
    UsePositionOffset=false
    Write Material Random=true
   )
   &AddVelocity(
    Cone Angle=33
    Cone Axis=<0 0.25 1>
    Module.UseConeFalloff=false
    Module.UseVelDistribution=false
    Override Seed=false
    Velocity Mode=::NewEnumerator2
    Velocity Speed=&FloatFromCurve(
     FloatCurve=DataInterface() {propertyValues {Curve {keys [<at 0 0.5>,<at 0.333 1>,<at 0.666 0.1>,<at 3 0>]}}}
     CurveIndex=$Emitter.LoopedAge
     Scale Curve=&RandomRangeFloat(Minimum=444, Maximum=999)
    )
   ) {enabled=false}
   &ShapeLocation(MaskSpawnGroup=false, Offset Mode=::NewEnumerator0, Override Seed=false, Sphere Radius=8, UseOwnerScale=true)
   &AddVelocity(
    Cone Axis=<-1 0 0>
    Module.UseConeFalloff=false
    Module.UseVelDistribution=false
    Override Seed=false
    Rotation Angle Type:ENiagara_AngleInput=::NewEnumerator1
    Rotation Mode=::NewEnumerator2
    Velocity Mode=::NewEnumerator2
    Velocity Speed=&RandomRangeFloat(Minimum=400, Maximum=700)
    Yaw / Pitch / Roll:Vector3f=<-0.25 0 0>
   ) {enabled=false, module=AddVelocity001}
   &AddVelocity(
    Module.Use Falloff Distance=false
    Origin Offset=<0 -15 0>
    Override Seed=false
    Velocity Mode=::NewEnumerator1
    Velocity Speed=&RandomRangeFloat(Minimum=200, Maximum=300)
   ) {module=AddVelocity002}
  ]
  particleUpdateScript [
   &ParticleState()
   &AccelerationForce() {enabled=false}
   &Drag(Drag=8, Use Linear Drag=true, Use Rotational Drag=false)
   &CurlNoiseForce(Noise Frequency=0.1, Noise Strength=1000, Pan Noise=true, Pan Noise Field=<0 0 1>)
   &AccelerationForce(Acceleration=<0 0 250>) {module=AccelerationForce001}
   &SolveForcesAndVelocity(Clamp Velocity=false, Limit Acceleration=false)
   &ScaleColor(
    Scale RGB=&MakeVectorFromLinearColorRGB(Color=&ColorFromCurve(ColorCurve=DataInterface() {propertyValues {
        AlphaCurve {keys [<at 0 1 RCIM_Cubic>]}
        BlueCurve {keys [<at 0.0265 0>,<at 0.1089 0 RCIM_Cubic>,<at 1 0 RCIM_Cubic>]}
        GreenCurve {keys [<at 0.0265 0>,<at 0.1089 1.8333 RCIM_Cubic>,<at 1 0 RCIM_Cubic>]}
        RedCurve {keys [<at 0.0265 0.2>, {arriveTangent=0.8, interpMode=RCIM_Cubic, time=0.1089, value=10}, <at 1 1 RCIM_Cubic>]}
       }}))
    ScaleA=true
    ScaleRGB=true
   )
   &ScaleSpriteSizeBySpeed(
    Max Scale Factor=<1.666 13>
    Min Scale Factor=<1.333 1.333>
    Sample Scale Factor By Curve=true
    Scale Factor Curve=DataInterface() {propertyValues {Curve {keys [<at 0 0>, {arriveTangent=1.4989, interpMode=RCIM_Cubic, tangentMode=RCTM_Break, time=1, value=1}]}}}
    Velocity Threshold=999
   )
   &ScaleSpriteSize(Uniform Curve Sprite Scale=DataInterface() {propertyValues {Curve {keys [
        {tangentMode=RCTM_User, time=0, value=1}
        {arriveTangent=-0.0405, interpMode=RCIM_Cubic, leaveTangent=-0.0405, tangentMode=RCTM_User, time=0.7009, value=1}
        {arriveTangent=-8.0089, interpMode=RCIM_Cubic, leaveTangent=-8.0089, tangentMode=RCTM_User, time=1, value=0}
       ]}}})
  ]
  rendererClasses [&NiagaraSpriteRendererProperties]
  renderers [{rendererClass=&NiagaraSpriteRendererProperties}]
 }
 Igni_Burn {
  emitterUpdateScript [
   &EmitterState(
    Inactive Response=::NewEnumerator1
    Life Cycle Mode=::NewEnumerator1
    Loop Behavior=::NewEnumerator1
    Loop Delay=&Add_Float(A=$System.FX_Duration, B=-1.04)
    UseLoopDelay=true
   )
   &SpawnBurst_Instantaneous(Use Spawn Probability=false, UseLoopCountLimit=false)
  ]
  particleSpawnScript [
   &InitializeParticle(
    Color Mode=::NewEnumerator0
    Lifetime=0.7
    Sprite Size Mode=::NewEnumerator3
    Uniform Sprite Size=35
    UsePositionOffset=false
    Write Material Random=true
   )
   &CameraOffset(Camera Offset Amount=5)
   &SpriteFacingAndAlignment(Sprite Alignment=<0 0 1>, Sprite Facing=<0 1 0>, WriteAlignment=true, WriteFacing=true)
   &DynamicMaterialParameters(
    Index 0 Param 2=$System.WPO_Wiggle_Freq1
    Index 0 Param 3=$System.WPO_Wiggle_Amp1
    Index 0 Param 4=$System.WPO_Wiggle_Amp2
    Module.Param0WriteEnabled=false
    Module.Param1WriteEnabled=true
    Module.Param2WriteEnabled=true
    Module.Param3WriteEnabled=true
   ) {enabled=false, module=DynamicMaterialParameters001}
  ]
  particleUpdateScript [
   &ParticleState()
   &ScaleSpriteSize(Uniform Curve Sprite Scale=DataInterface() {propertyValues {Curve {keys [
        {arriveTangent=0.3516, interpMode=RCIM_Cubic, leaveTangent=0.3516, tangentMode=RCTM_User, time=0, value=0.8}
        {arriveTangent=0.0542, interpMode=RCIM_Cubic, leaveTangent=0.0542, tangentMode=RCTM_User, time=0.65, value=1}
       ]}}})
   &ScaleColor(
    Scale Alpha=&FloatFromCurve(FloatCurve=DataInterface() {propertyValues {Curve {keys [
         {arriveTangent=0.5924, arriveTangentWeight=1.1112, interpMode=RCIM_Cubic, leaveTangent=0.5924, tangentMode=RCTM_User, time=0.25, value=0.1}
         {arriveTangent=0.2859, interpMode=RCIM_Cubic, leaveTangent=0.2859, tangentMode=RCTM_User, time=1, value=1}
        ]}}})
    Scale RGB=&MakeVectorFromLinearColorRGB(Color=&ColorFromCurve(ColorCurve=DataInterface() {propertyValues {
        AlphaCurve {keys [<at 0 1 RCIM_Cubic>]}
        BlueCurve {keys [
          <at 0 0.03>
          <at 0.3011 0.102>
          {arriveTangent=-0.05, interpMode=RCIM_Cubic, leaveTangent=-0.05, time=0.3556, value=0.4706}
          <at 0.6003 0>
          <at 1 0>
         ]}
        GreenCurve {keys [
          <at 0 0.03>
          <at 0.3011 0.5804>
          {arriveTangent=-0.0197, interpMode=RCIM_Cubic, leaveTangent=-0.0197, time=0.3556, value=2.9894}
          <at 0.6003 0.0182>
          <at 1 0>
         ]}
        RedCurve {keys [
          <at 0 0.03>
          <at 0.3011 2>
          {arriveTangent=1.6157, interpMode=RCIM_Cubic, leaveTangent=1.6157, time=0.3556, value=10}
          <at 0.6003 1>
          <at 1 0>
         ]}
       }}))
    ScaleA=true
    ScaleRGB=true
   )
  ]
  rendererClasses [&NiagaraSpriteRendererProperties, &NiagaraLightRendererProperties]
  renderers [
   {rendererClass=&NiagaraSpriteRendererProperties}
   {rendererClass=&NiagaraLightRendererProperties, rendererIndex=1}
  ]
 }
 Igni_Main {
  emitterUpdateScript [
   &EmitterState()
   &SpawnBurst_Instantaneous(Use Spawn Probability=false, UseLoopCountLimit=false)
  ]
  particleSpawnScript [
   &InitializeParticle(
    Color Mode=::NewEnumerator0
    Lifetime=$System.FX_Duration
    Sprite Size Mode=::NewEnumerator3
    Uniform Sprite Size=33
    UsePositionOffset=false
    Write Material Random=true
   )
   /SetVariables(Particles.Color_Light:LinearColor=<rgba 1.2 0.3947 0 1>) {isSetParameters=true}
   /SetVariables() {isSetParameters=true}
   /SetVariables(Particles.Glow_Size:Vector2f=&Multiply_Vector2D_ByFloat(Float=5, Vector2D=$Particles.SpriteSize)) {isSetParameters=true}
   &SpriteFacingAndAlignment(Sprite Alignment=<0 0 1>, Sprite Facing=<0 1 0>, WriteAlignment=true, WriteFacing=true)
   &DynamicMaterialParameters(
    Index 0 Param 2=$System.WPO_Wiggle_Freq1
    Index 0 Param 3=$System.WPO_Wiggle_Amp1
    Index 0 Param 4=$System.WPO_Wiggle_Amp2
    Module.Param0WriteEnabled=false
    Module.Param1WriteEnabled=true
    Module.Param2WriteEnabled=true
    Module.Param3WriteEnabled=true
   ) {enabled=false, module=DynamicMaterialParameters001}
  ]
  particleUpdateScript [
   &ParticleState()
   &ScaleSpriteSize(Uniform Curve Sprite Scale=DataInterface() {propertyValues {Curve {keys [
        {arriveTangent=0.4753, interpMode=RCIM_Cubic, leaveTangent=0.4753, tangentMode=RCTM_User, time=0, value=0.8}
        {arriveTangent=-0.0405, interpMode=RCIM_Cubic, leaveTangent=-0.0405, tangentMode=RCTM_User, time=1, value=1}
       ]}}}) {enabled=false}
   &ScaleColor(
    Scale Alpha=&FloatFromCurve(FloatCurve=DataInterface() {propertyValues {Curve {keys [
         {arriveTangent=0.0475, interpMode=RCIM_Cubic, leaveTangent=0.0475, tangentMode=RCTM_User, time=0.24, value=0}
         {arriveTangentWeight=1.1112, interpMode=RCIM_Cubic, tangentMode=RCTM_User, time=0.87, value=0}
         {arriveTangent=4.2376, interpMode=RCIM_Cubic, leaveTangent=4.2376, tangentMode=RCTM_User, time=0.97, value=1}
        ]}}})
    Scale RGB=&MakeVectorFromLinearColorRGB(Color=&ColorFromCurve(ColorCurve=DataInterface() {propertyValues {
        AlphaCurve {keys [<at 0 1 RCIM_Cubic>]}
        BlueCurve {keys [<at 0 0.01 RCIM_Cubic>, <at 0.06 0.2>, {arriveTangent=-0.2128, interpMode=RCIM_Cubic, leaveTangent=-0.2128, time=0.915, value=0.2}, <at 1 0>]}
        GreenCurve {keys [
          <at 0 0.01 RCIM_Cubic>
          <at 0.06 3.4503>
          {arriveTangent=-3.6696, interpMode=RCIM_Cubic, leaveTangent=-3.6696, time=0.915, value=3.4503}
          <at 1 0.0009>
         ]}
        RedCurve {keys [
          <at 0 0.01 RCIM_Cubic>
          <at 0.06 10>
          {arriveTangent=-10.6106, interpMode=RCIM_Cubic, leaveTangent=-10.6106, time=0.915, value=10}
          <at 1 0.026>
         ]}
       }}))
    ScaleA=true
    ScaleRGB=true
   )
   &DynamicMaterialParameters(
    Index 0 Param 1=&FloatFromCurve(FloatCurve=DataInterface() {propertyValues {Curve {keys [
         <at 0 0 RCIM_Cubic>
         {arriveTangent=-0.0035, interpMode=RCIM_Cubic, leaveTangent=-0.0035, tangentMode=RCTM_User, time=0.2, value=1}
         {arriveTangent=0.0033, interpMode=RCIM_Cubic, leaveTangent=0.0033, tangentMode=RCTM_User, time=0.84, value=1}
         <at 0.95 0 RCIM_Cubic>
        ]}}})
    Module.Param0WriteEnabled=true
    Module.Param1WriteEnabled=false
    Module.Param2WriteEnabled=false
    Module.Param3WriteEnabled=false
   )
   /SetVariables(Particles.Color_Light:LinearColor=&MakeLinearColorFromVectorAndFloat(
     "Vector (RGB)"=&MakeVectorFromLinearColorRGB(Color=$Particles.Color_Light)
     "Float (Alpha)"=&Multiply_Float(
      A=&FloatFromCurve(FloatCurve=DataInterface() {propertyValues {Curve {keys [
           {interpMode=RCIM_Cubic, tangentMode=RCTM_User, time=0, value=0}
           {arriveTangent=-0.0392, interpMode=RCIM_Cubic, leaveTangent=-0.0392, tangentMode=RCTM_User, time=0.05, value=1}
           {interpMode=RCIM_Cubic, tangentMode=RCTM_User, time=0.85, value=1}
           {interpMode=RCIM_Cubic, tangentMode=RCTM_User, time=1, value=0}
          ]}}})
      B=&FloatFromCurve(FloatCurve=DataInterface() {propertyValues {Curve {keys [<at 0 1>,<at 0.03 0.95>,<at 0.07 0.7>,<at 0.1 1>,<at 0.16 0.9>,<at 0.2 0.8>,<at 0.24 0.5>,<at 0.3 0.7>,<at 0.32 0.6>,<at 0.36 0.95>,<at 0.43 0.85>,<at 0.48 1>,<at 0.49 0.8>,<at 0.58 1>,<at 0.63 0.6>,<at 0.7 0.95>,<at 0.72 0.78>,<at 0.76 0.9>,<at 0.8 0.85>,<at 0.9 0.95>,<at 1 1>]}}})
     )
    )) {isSetParameters=true}
   /SetVariables(Particles.Color_Glow:LinearColor=&MakeLinearColorFromVectorAndFloat(
     "Vector (RGB)"=&MakeVectorFromLinearColorRGB(Color=&ColorFromCurve(ColorCurve=DataInterface() {propertyValues {AlphaCurve {keys []}, BlueCurve {keys [<at 0.9097 0>]}, GreenCurve {keys [<at 0.9097 9.9426>]}, RedCurve {keys [<at 0.9097 30>]}}}))
     "Float (Alpha)"=&FloatFromCurve(
      FloatCurve=DataInterface() {propertyValues {Curve {keys [<at 0 0>,<at 0.1 0.2>,<at 0.92 0.2>,<at 0.925 0.8>,<at 0.95 0>]}}}
      Scale Curve=0.03
     )
    )) {isSetParameters=true}
   /SetVariables(Particles.Position=&AddVectorToPosition(
     Position=/Niagara/DynamicInputs/Helpers/SimulationPosition()
     Vector=&VectorFromCurve(VectorCurve=DataInterface() {propertyValues {
        XCurve {keys [<at 0 0>,<at 1 0>]}
        YCurve {keys [
          {arriveTangent=1069.6884, interpMode=RCIM_Cubic, leaveTangent=1069.6884, tangentMode=RCTM_User, time=0, value=-10}
          {interpMode=RCIM_Cubic, tangentMode=RCTM_User, time=0.04, value=0}
         ]}
        ZCurve {keys [<at 0 0>,<at 1 0>]}
       }})
    )) {isSetParameters=true}
  ]
  rendererClasses [&NiagaraSpriteRendererProperties, &NiagaraLightRendererProperties]
  renderers [
   {rendererClass=&NiagaraSpriteRendererProperties}
   {rendererClass=&NiagaraLightRendererProperties, rendererIndex=1}
   {rendererClass=&NiagaraSpriteRendererProperties, rendererIndex=2}
  ]
 }
 Igni_SideMesh {
  emitterUpdateScript [
   &EmitterState()
   &SpawnBurst_Instantaneous(Use Spawn Probability=false, UseLoopCountLimit=false)
  ]
  particleSpawnScript [
   &InitializeParticle(
    Color Mode=::NewEnumerator0
    Lifetime=$System.FX_Duration
    Mesh Scale=&Multiply_VectorByFloat(Vector=<0.17 0.2 0.2>, Float=0.83)
    Mesh Scale Mode=::NewEnumerator4
    UsePositionOffset=false
    Write Material Random=true
   )
   &InitialMeshOrientation(Mesh Orientation Mode=::NewEnumerator2, Orientation Vector=<0 1 0>, Use Rotation Vector=true)
   &DynamicMaterialParameters(
    Index 0 Param 2=$System.WPO_Wiggle_Freq1
    Index 0 Param 3=$System.WPO_Wiggle_Amp1
    Index 0 Param 4=$System.WPO_Wiggle_Amp2
    Module.Param0WriteEnabled=false
    Module.Param1WriteEnabled=true
    Module.Param2WriteEnabled=true
    Module.Param3WriteEnabled=true
   ) {enabled=false, module=DynamicMaterialParameters001}
  ]
  particleUpdateScript [
   &ParticleState()
   &ScaleColor(
    Scale Alpha=&FloatFromCurve(FloatCurve=DataInterface() {propertyValues {Curve {keys [{arriveTangent=0.0475, time=0, value=0}, <at 0.1 1>, <at 0.85 1>, {arriveTangentWeight=1.1112, time=0.92, value=0}]}}})
    Scale RGB=&MakeVectorFromLinearColorRGB(Color=&ColorFromCurve(ColorCurve=DataInterface() {propertyValues {
        AlphaCurve {keys [<at 0 1 RCIM_Cubic>]}
        BlueCurve {keys [
          <at 0 0.01 RCIM_Cubic>
          <at 0.0835 0.21>
          {arriveTangent=-0.2291, interpMode=RCIM_Cubic, leaveTangent=-0.2291, time=0.895, value=0.21}
          <at 1 0>
         ]}
        GreenCurve {keys [
          <at 0 0.01 RCIM_Cubic>
          <at 0.0835 2.0207>
          {arriveTangent=-2.2037, interpMode=RCIM_Cubic, leaveTangent=-2.2037, time=0.895, value=2.0207}
          <at 1 0.0009>
         ]}
        RedCurve {keys [
          <at 0 0.01 RCIM_Cubic>
          <at 0.0835 7>
          {arriveTangent=-7.6092, interpMode=RCIM_Cubic, leaveTangent=-7.6092, time=0.895, value=7}
          <at 1 0.026>
         ]}
       }}))
    ScaleA=true
    ScaleRGB=true
   )
   /SetVariables(Particles.Position=&AddVectorToPosition(
     Position=&SimulationPosition()
     Vector=&VectorFromCurve(VectorCurve=DataInterface() {propertyValues {
        XCurve {keys [<at 0 0>,<at 1 0>]}
        YCurve {keys [
          {arriveTangent=1069.6884, interpMode=RCIM_Cubic, leaveTangent=1069.6884, tangentMode=RCTM_User, time=0, value=-10}
          {interpMode=RCIM_Cubic, tangentMode=RCTM_User, time=0.04, value=0}
         ]}
        ZCurve {keys [<at 0 0>,<at 1 0>]}
       }})
    )) {isSetParameters=true}
  ]
  rendererClasses [&NiagaraMeshRendererProperties]
  renderers [{rendererClass=&NiagaraMeshRendererProperties, rendererIndex=0}]
 }
 Pulse {
  emitterUpdateScript [
   &EmitterState(
    Inactive Response=::NewEnumerator1
    Life Cycle Mode=::NewEnumerator1
    Loop Behavior=::NewEnumerator1
    Loop Delay=0.1
    Loop Duration=&Add_Float(A=$System.FX_Duration, B=-1)
    UseLoopDelay=true
   )
   &SpawnRate(
    SpawnRate=&FloatFromCurve(
     FloatCurve=DataInterface() {propertyValues {Curve {keys [
         {arriveTangent=-2.6167, interpMode=RCIM_Cubic, leaveTangent=-2.6167, tangentMode=RCTM_User, time=0.8, value=1}
         {arriveTangent=-0.1055, interpMode=RCIM_Cubic, leaveTangent=-0.1055, tangentMode=RCTM_User, time=0.95, value=0}
        ]}}}
     CurveIndex=$Emitter.NormalizedLoopAge
     Scale Curve=0.8
    )
    Use Spawn Probability=true
   )
   &SpawnBurst_Instantaneous(Use Spawn Probability=false, UseLoopCountLimit=false) {enabled=false}
  ]
  particleSpawnScript [&InitializeParticle(
    Color=&RandomRangeLinearColor(Minimum=<rgba 2 0.6082 0.1454 0.02>, Maximum=<rgba 2 0.306 0 0.02>)
    Lifetime=&RandomRangeFloat(Minimum=1.2, Maximum=1.6)
    Position Offset=<0 0 -34>
    Sprite Rotation Mode=::NewEnumerator2
    Sprite Size Mode=::NewEnumerator3
    Uniform Sprite Size=&RandomRangeFloat(Minimum=50, Maximum=60)
    UsePositionOffset=false
    Write Material Random=true
   )]
  particleUpdateScript [
   &ParticleState()
   &ScaleSpriteSize(Uniform Curve Sprite Scale=DataInterface() {propertyValues {Curve {keys [{arriveTangent=1.4169, interpMode=RCIM_Cubic, leaveTangent=1.4169, tangentMode=RCTM_User, time=0, value=0.5}, <at 1 1 RCIM_Cubic>]}}})
   &Drag(Drag=1.7, Use Linear Drag=true, Use Rotational Drag=true) {enabled=false}
   &SolveForcesAndVelocity(Clamp Velocity=false, Limit Acceleration=false) {enabled=false}
   &ScaleColor(
    Scale RGB=&MakeVectorFromLinearColorRGB(Color=&ColorFromCurve(ColorCurve=DataInterface() {propertyValues {
        AlphaCurve {keys [<at 0 1 RCIM_Cubic>]}
        BlueCurve {keys [<at 0 0>,<at 0.068 0.516 RCIM_Cubic>,<at 0.9242 0 RCIM_Cubic>]}
        GreenCurve {keys [<at 0 0>,<at 0.068 2.813 RCIM_Cubic>,<at 0.9242 0 RCIM_Cubic>]}
        RedCurve {keys [<at 0 0>,<at 0.068 5 RCIM_Cubic>,<at 0.9242 0 RCIM_Cubic>]}
       }}))
    ScaleA=true
    ScaleRGB=true
   )
   &DynamicMaterialParameters(
    Index 0 Param 1=&FloatFromCurve(
     FloatCurve=DataInterface() {propertyValues {Curve {keys [
         <at 0 0 RCIM_Cubic>
         {arriveTangent=0.0033, interpMode=RCIM_Cubic, leaveTangent=0.0033, tangentMode=RCTM_User, time=0.3, value=1}
         <at 1 0 RCIM_Cubic>
        ]}}}
     Scale Curve=0.05
    )
    Module.Param0WriteEnabled=true
    Module.Param1WriteEnabled=true
    Module.Param2WriteEnabled=true
    Module.Param3WriteEnabled=true
   )
  ]
  rendererClasses [&NiagaraSpriteRendererProperties]
  renderers [{rendererClass=&NiagaraSpriteRendererProperties}]
 }
 Smoke_Constant {
  bEnabled=false
  emitterUpdateScript [
   &EmitterState(
    Inactive Response=::NewEnumerator1
    Life Cycle Mode=::NewEnumerator1
    Loop Behavior=::NewEnumerator1
    Loop Duration=$System.FX_Duration
    UseLoopDelay=false
   )
   &SpawnRate(
    SpawnRate=&FloatFromCurve(
     FloatCurve=DataInterface() {propertyValues {Curve {keys [
         {arriveTangent=-2.6167, interpMode=RCIM_Cubic, leaveTangent=-2.6167, tangentMode=RCTM_User, time=0.5, value=1}
         {arriveTangent=-1.121, interpMode=RCIM_Cubic, leaveTangent=-1.121, tangentMode=RCTM_User, time=0.95, value=0}
        ]}}}
     CurveIndex=$Emitter.NormalizedLoopAge
     Scale Curve=28
    )
    Use Spawn Probability=false
   ) {module=SpawnRate001}
  ]
  particleSpawnScript [
   &InitializeParticle(
    Color=&UniformRangedLinearColor(Minimum=<rgba 0.5 0.2913 0.1705 0.25>, Maximum=<rgba 0.1 0.0815 0.07 0.1>)
    Lifetime=&UniformRangedFloat(Minimum=0.5, Maximum=1.7)
    Sprite Rotation Mode=::NewEnumerator2
    Sprite Size Mode=::NewEnumerator3
    Uniform Sprite Size=&RandomRangeFloat(Minimum=40, Maximum=80)
    UsePositionOffset=false
    Write Material Random=true
   ) {module=InitializeParticle001}
   /SetVariables(Particles.InitSpriteRotRate=&RandomRangeFloat(Minimum=-2, Maximum=2)) {isSetParameters=true}
   &AddVelocity(
    Rotation Angle Type:ENiagara_AngleInput=::NewEnumerator1
    Rotation Mode=::NewEnumerator2
    Velocity=&RandomRangeVector(Minimum=<0 70 5>, Maximum=<0 100 15>)
   ) {module=AddVelocity001}
  ]
  particleUpdateScript [
   &ParticleState()
   &SpriteRotationRate(Rotation Rate=&Multiply_Float(
     A=$Particles.InitSpriteRotRate
     B=&FloatFromCurve(
      FloatCurve=DataInterface() {propertyValues {Curve {keys [
          {arriveTangent=-4.0107, interpMode=RCIM_Cubic, leaveTangent=-4.0107, tangentMode=RCTM_User, time=-0.0046, value=1.0113}
          {arriveTangent=-0.2496, interpMode=RCIM_Cubic, leaveTangent=-0.2496, tangentMode=RCTM_User, time=0.5, value=0.16}
         ]}}}
      CurveIndex=$Emitter.Age
      Scale Curve=20
     )
    ))
   &ScaleColor(
    Scale Alpha=&Multiply_Float(
     A=&FloatFromCurve(FloatCurve=DataInterface() {propertyValues {Curve {keys [
          {arriveTangent=9.0725, interpMode=RCIM_Cubic, leaveTangent=9.0725, tangentMode=RCTM_User, time=0, value=0}
          {arriveTangent=-0.1032, interpMode=RCIM_Cubic, leaveTangent=-0.1032, tangentMode=RCTM_User, time=0.1, value=1}
          {arriveTangent=-0.0405, interpMode=RCIM_Cubic, leaveTangent=-0.0405, tangentMode=RCTM_User, time=1, value=0}
         ]}}})
     B=&FN_Select_Float_from_Time_Of_Day(Morning=0.85, Day=0.8, Evening=1, Night=1)
    )
    Scale RGB=&Multiply_VectorByFloat(
     Vector=&MakeVectorFromLinearColorRGB(Color=&ColorFromCurve(ColorCurve=DataInterface() {propertyValues {
         AlphaCurve {keys [<at 0 1 RCIM_Cubic>]}
         BlueCurve {keys [<at 0 0.2281 RCIM_Cubic>,<at 0.3938 0.19 RCIM_Cubic>]}
         GreenCurve {keys [<at 0 0.6969 RCIM_Cubic>,<at 0.3938 0.19 RCIM_Cubic>]}
         RedCurve {keys [<at 0 1.3 RCIM_Cubic>,<at 0.3938 0.19 RCIM_Cubic>]}
        }}))
     Float=&FN_Select_Float_from_Time_Of_Day(Morning=0.6, Day=1, Evening=1, Night=1)
    )
    ScaleA=true
    ScaleRGB=true
   )
   &ScaleSpriteSize(Scale Factor=&Vector2DFromFloat(Value=&FloatFromCurve(FloatCurve=DataInterface() {propertyValues {Curve {keys [
          {arriveTangent=1.2912, interpMode=RCIM_Cubic, leaveTangent=1.2912, tangentMode=RCTM_User, time=0, value=0.2}
          {arriveTangent=0.1638, interpMode=RCIM_Cubic, leaveTangent=0.1638, tangentMode=RCTM_User, time=1, value=1}
         ]}}})))
   &DynamicMaterialParameters(
    Index 0 Param 1=&Add_Float(
     A=$Particles.RandomFloat
     B=&FloatFromCurve(FloatCurve=DataInterface() {propertyValues {Curve {keys [
          {arriveTangent=2.7533, interpMode=RCIM_Cubic, leaveTangent=2.847, tangentMode=RCTM_Break, time=0, value=0}
          {arriveTangent=1.3581, interpMode=RCIM_Cubic, leaveTangent=1.3581, tangentMode=RCTM_User, time=0.2558, value=0.5125}
          {arriveTangent=0.0529, interpMode=RCIM_Cubic, leaveTangent=0.2021, tangentMode=RCTM_User, time=1, value=1}
         ]}}})
    )
    Index 0 Param 3=&FloatFromCurve(FloatCurve=DataInterface() {propertyValues {Curve {keys [{arriveTangent=-53.9199, time=0, value=0}, {leaveTangent=0.2723, time=0.25, value=0}]}}})
    Index 0 Param 4=&FloatFromCurve(FloatCurve=DataInterface() {propertyValues {Curve {keys [<at 0 1>,<at 1 0>]}}})
    Index 1 Param 1=&FloatFromCurve(FloatCurve=DataInterface() {propertyValues {Curve {keys [<at 0 0>, {arriveTangent=4.2257, interpMode=RCIM_Cubic, leaveTangent=4.2257, tangentMode=RCTM_User, time=0.25, value=0}]}}})
    Index 1 Param 2=&FloatFromCurve(FloatCurve=DataInterface() {propertyValues {Curve {keys [<at 0 0>]}}})
    Index 1 Param 3=&FloatFromCurve(FloatCurve=DataInterface() {propertyValues {Curve {keys [{interpMode=RCIM_Cubic, leaveTangent=6.3917, tangentMode=RCTM_Break, time=0, value=0}]}}})
    Index 1 Param 4=&FloatFromCurve(FloatCurve=DataInterface() {propertyValues {Curve {keys [{arriveTangent=-0.136, leaveTangent=-0.136, time=0, value=1}]}}})
    Index 2 Param 1=&FloatFromCurve(FloatCurve=DataInterface() {propertyValues {Curve {keys [
         {arriveTangent=2.646, interpMode=RCIM_Cubic, leaveTangent=2.646, tangentMode=RCTM_User, time=0, value=0}
         {leaveTangent=-0.0719, time=1, value=1}
        ]}}})
    Index 2 Param 2=&FloatFromCurve(
     FloatCurve=DataInterface() {propertyValues {Curve {keys [
         {arriveTangent=1.9056, time=0, value=0.2}
         {arriveTangent=0.0259, interpMode=RCIM_Cubic, leaveTangent=0.0259, tangentMode=RCTM_User, time=1, value=1.4}
        ]}}}
     Scale Curve=&RandomRangeFloat(Minimum=0.5, Maximum=1)
    )
    Index 2 Param 3=&FloatFromCurve(FloatCurve=DataInterface() {propertyValues {Curve {keys [
         {interpMode=RCIM_Cubic, leaveTangent=2.3064, tangentMode=RCTM_Break, time=0, value=1}
         {interpMode=RCIM_Cubic, leaveTangent=11.1233, tangentMode=RCTM_Break, time=1, value=2}
        ]}}})
    Index 2 Param 4=&FloatFromCurve(
     FloatCurve=DataInterface() {propertyValues {Curve {keys [
         {arriveTangent=-0.6867, interpMode=RCIM_Cubic, leaveTangent=-0.6867, tangentMode=RCTM_User, time=0, value=1}
         {interpMode=RCIM_Cubic, leaveTangent=-1.2956, tangentMode=RCTM_Break, time=1, value=0.75}
        ]}}}
     Scale Curve=1.6
    )
    Module.Param0WriteEnabled=true
    Module.Param0WriteEnabled1=true
    Module.Param0WriteEnabled2=true
    Module.Param1WriteEnabled=false
    Module.Param1WriteEnabled1=true
    Module.Param1WriteEnabled2=true
    Module.Param2WriteEnabled=true
    Module.Param2WriteEnabled1=true
    Module.Param2WriteEnabled2=true
    Module.Param3WriteEnabled=true
    Module.Param3WriteEnabled1=true
    Module.Param3WriteEnabled2=true
    Write Parameter Index 1=true
    Write Parameter Index 2=true
   )
   &Drag(Drag=2, Use Linear Drag=true, Use Rotational Drag=false)
   &GravityForce(Gravity=<0 0 20>)
   &SolveForcesAndVelocity(Clamp Velocity=false, Limit Acceleration=false)
   &ScaleColor(
    Scale RGB=&VectorFromFloat(Value=&FloatFromCurve(
      FloatCurve=DataInterface() {propertyValues {Curve {keys [{arriveTangent=1.0448, time=0, value=0}, {arriveTangent=-0.0405, leaveTangent=-0.0405, time=1, value=1}]}}}
      CurveIndex=$Emitter.Age
      Scale Curve=2
     ))
    ScaleA=true
    ScaleRGB=true
   ) {module=ScaleColor002}
  ]
  rendererClasses [&NiagaraSpriteRendererProperties]
  renderers [
   {rendererClass=&NiagaraSpriteRendererProperties}
   {rendererClass=&NiagaraSpriteRendererProperties, rendererIndex=1}
  ]
 }
 Sphere {
  emitterUpdateScript [
   &EmitterState()
   &SpawnBurst_Instantaneous(Use Spawn Probability=false, UseLoopCountLimit=false)
  ]
  particleSpawnScript [&InitializeParticle(
    Color=<rgba 3 1.05 0 1>
    Lifetime=0.6
    Sprite Size Mode=::NewEnumerator3
    Uniform Sprite Size=58
    UsePositionOffset=false
    Write Material Random=true
   )]
  particleUpdateScript [
   &ParticleState()
   &ScaleSpriteSize(Uniform Curve Sprite Scale=DataInterface() {propertyValues {Curve {keys [
        {arriveTangent=4.1785, interpMode=RCIM_Cubic, leaveTangent=4.1785, tangentMode=RCTM_User, time=0, value=0}
        {arriveTangent=0.9995, interpMode=RCIM_Cubic, leaveTangent=0.9995, tangentMode=RCTM_User, time=0.2223, value=0.6825}
        {arriveTangent=-0.0405, interpMode=RCIM_Cubic, leaveTangent=-0.0405, tangentMode=RCTM_User, time=1, value=1}
       ]}}})
   &ScaleColor(
    Scale Alpha=&FloatFromCurve(FloatCurve=DataInterface() {propertyValues {Curve {keys [
         {interpMode=RCIM_Cubic, tangentMode=RCTM_User, time=0, value=0}
         {arriveTangent=-17.9937, interpMode=RCIM_Cubic, leaveTangent=-17.9937, leaveTangentWeight=0.9539, tangentMode=RCTM_User, time=0.05, value=3}
         <at 0.15 1>
         {arriveTangentWeight=1.1112, interpMode=RCIM_Cubic, tangentMode=RCTM_User, time=1, value=0}
        ]}}})
    Scale RGB=&MakeVectorFromLinearColorRGB(Color=&ColorFromCurve(ColorCurve=DataInterface() {propertyValues {
        AlphaCurve {keys []}
        BlueCurve {keys [<at 0.0189 10 RCIM_Cubic>,<at 0.0605 1>,<at 0.6068 1>,<at 1 0.3>]}
        GreenCurve {keys [<at 0.0189 10 RCIM_Cubic>,<at 0.0605 1>,<at 0.6068 1>,<at 1 0.3>]}
        RedCurve {keys [<at 0.0189 10 RCIM_Cubic>,<at 0.0605 1>,<at 0.6068 1>,<at 1 0.3>]}
       }}))
    ScaleA=true
    ScaleRGB=true
   )
  ]
  rendererClasses [&NiagaraSpriteRendererProperties]
  renderers [{rendererClass=&NiagaraSpriteRendererProperties}]
 }
}
summary {
 systemName=NS_Emote_MagicMeadow_Main
 emitters [
  {emitterName=Sphere, rendererClasses [&NiagaraSpriteRendererProperties]}
  {emitterName=Igni_Main, rendererClasses [&NiagaraSpriteRendererProperties, &NiagaraLightRendererProperties]}
  {emitterName=Igni_Burn, rendererClasses [&NiagaraSpriteRendererProperties, &NiagaraLightRendererProperties]}
  {emitterName=Embers_Burst, rendererClasses [&NiagaraSpriteRendererProperties]}
  {emitterName=Flames, rendererClasses [&NiagaraSpriteRendererProperties]}
  {bEnabled=false, emitterName=Forward_Sparks, rendererClasses [&NiagaraSpriteRendererProperties]}
  {emitterName=Igni_SideMesh, rendererClasses [&NiagaraMeshRendererProperties]}
  {emitterName=Pulse, rendererClasses [&NiagaraSpriteRendererProperties]}
  {emitterName=Forward_Flames_Soft, rendererClasses [&NiagaraSpriteRendererProperties]}
  {bEnabled=false, emitterName=Embers_Forward002, rendererClasses [&NiagaraSpriteRendererProperties]}
  {bEnabled=false, emitterName=Smoke_Constant, rendererClasses [&NiagaraSpriteRendererProperties]}
  {emitterName=Forward_Sparks_New, rendererClasses [&NiagaraSpriteRendererProperties]}
 ]
 userVariables [User.SkeletalMesh {
   type=NiagaraDataInterfaceSkeletalMesh
   defaultValue {
    struct=NiagaraExt_VariableValue_DataInterface
    value {dataInterface=&self:NiagaraDataInterfaceSkeletalMesh_0}
   }
  }]
}
